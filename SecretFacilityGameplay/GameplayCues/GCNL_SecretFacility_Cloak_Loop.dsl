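; Looping gameplay cue that drives player cloak visuals in Secret Facility, extending the FortGameCueNotifyLoop_Cloak behavior.
; An async wait for the holster-weapon tag sets bCloaked, which gates exit effects.
; Applying the material override can hide attached FX as a backup, and removing it restores them.
; When the timer override is enabled, invisibility goo overlaps clear the expire tell, and leaving restarts its timer for roughly the duration's final quarter.
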(event Custom|BindA51GooOverlapEvents
  (params (TargetPlayer :type Actor%20Object%20Reference))
  (if (Variables|Default|GetExpireTellTimerOverride)
    (Collision|BindEventtoOnActorBeginOverlap TargetPlayer (delegate Custom|StandingOnA51Goo))
    (Collision|BindEventtoOnActorEndOverlap TargetPlayer (delegate Custom|EndStandingOnA51Goo))))

(event Custom|EndStandingOnA51Goo
  (params (OverlappedActor :type Actor%20Object%20Reference) (OtherActor :type Actor%20Object%20Reference))
  (if (or (== (Utilities|GetClass OtherActor) "/SecretFacilityGameplay/Gameplay/Actors/BGA_SecretFacility_InvisibilityGoo.BGA_SecretFacility_InvisibilityGoo_C") (== (Utilities|GetClass OtherActor) "/SecretFacilityGameplay/Gameplay/Actors/BGA_SecretFacility_InvisibilityGoo_NoLight.BGA_SecretFacility_InvisibilityGoo_NoLight_C"))
    (CallFunction|ClearExpireTell)
    (bind _returnvalue (Utilities|Time|SetTimerbyEvent (EventDispatchers|CreateEvent :!event BeginExpireTell) (- (ScalableFloat|GetValueatLevel (Variables|Balance|GetRow_Duration)) (Math|Conversions|ToFloat(Integer) (Math|Float|Round (* (ScalableFloat|GetValueatLevel (Variables|Balance|GetRow_Duration)) 0.25))))))
    (Variables|Default|SetExpireTellDelayTimer _returnvalue)
    (Development|PrintString 0 "STARTING A51 GOO EXPIRE TELL TIMER" false)))

(event Custom|StandingOnA51Goo
  (params (OverlappedActor :type Actor%20Object%20Reference) (OtherActor :type Actor%20Object%20Reference))
  (if (or (== (Utilities|GetClass OtherActor) "/SecretFacilityGameplay/Gameplay/Actors/BGA_SecretFacility_InvisibilityGoo.BGA_SecretFacility_InvisibilityGoo_C") (== (Utilities|GetClass OtherActor) "/SecretFacilityGameplay/Gameplay/Actors/BGA_SecretFacility_InvisibilityGoo_NoLight.BGA_SecretFacility_InvisibilityGoo_NoLight_C"))
    (CallFunction|ClearExpireTell)
    (Development|PrintString 0 "ENDING A51 GOO EXPIRE TELL TIMER" false)))

(event EventOnLoopingStart(EitherVFX-preferthisversion)
  (params (MyTarget :type Actor%20Object%20Reference) (Parameters :type Gameplay%20Cue%20Parameters%20Structure%20%28by%20ref%29) (ParticleComponents :type Array%20of%20FXSystem%20Component%20Object%20References) (AudioComponents :type Array%20of%20Audio%20Component%20Object%20References))
  (CallFunction|BindA51GooOverlapEvents MyTarget)
  (|CallParentFunction MyTarget Parameters ParticleComponents AudioComponents)
  (Variables|Default|SetbCloaked false)
  (Variables|FortGameCueNotifyLoop_Cloak|SetSpeedForMaxVis (ScalableFloat|GetValueatLevel (Variables|Balance|GetRow_SpeedForMaxVis)))
  (bind _asynctaskproxy (Ability|Tasks|WaitGameplayTagAddToActor (Variables|FortGameCueNotifyLoop_Cloak|GetTargetPlayer) "(TagName=\"Granted.Athena.FlexLegendMedallions.Cloak.HolsterWeapon\")")
    (:then
      (Variables|Default|SetWaitForHolsterAsyncAction _asynctaskproxy))
    (:Added
      (Variables|Default|SetbCloaked true))))

(event EventOnRemoval(EitherVFX-preferthisversion)
  (params (MyTarget :type Actor%20Object%20Reference) (Parameters :type Gameplay%20Cue%20Parameters%20Structure%20%28by%20ref%29) (ParticleComponents :type Array%20of%20FXSystem%20Component%20Object%20References) (AudioComponents :type Array%20of%20Audio%20Component%20Object%20References) (BurstCameraShakeInstance :type Camera%20Shake%20Base%20Object%20Reference) (BurstDecalInstance :type Decal%20Actor%20Object%20Reference))
  (bind _activatecloakfx (Variables|Default|GetActivateCloakFX))
  (bind _asfort_pawn (Utilities|Casting|CastToFortPawn MyTarget)
    (:then
      (Niagara|SpawnSystemAttached (Variables|FX|GetDeactivateFXTemplate) (Variables|/Script/Engine.Character|GetMesh _asfort_pawn) "root" "0, 0, 0" "0, 0, 0" "KeepRelativeOffset" true true "AutoRelease"))
    (:CastFailed
      _))
  (if (not (Variables|Default|GetbCloaked))
    ;[0.15 0 0.08505 0.5] If we haven't actually cloaked, cancel the exit effects
    (Components|Activation|Deactivate _activatecloakfx)
    (Components|DestroyComponent _activatecloakfx 0))
  (|CallParentFunction MyTarget Parameters ParticleComponents AudioComponents BurstCameraShakeInstance BurstDecalInstance)
  (Ability|Async|EndAction (Variables|Default|GetWaitForHolsterAsyncAction)))

(fn ApplyCloakMaterialOverride
  ;[0.04536 0.15 0.1105 0.5] When player becomes fully cloaked, hide any other FX attached to the player
  (|CallParentFunction)
  (CallFunction|HideAttachedFX))

(fn HideAttachedFX
  ;[0.1045 0.04764 0.15 0.5] Handles hiding FX that are attached to the player while cloaked
  ;;[0.3323 0.05088 0.04493 0.5] NOTE: Hiding FX should be handled by the pawn material override system. This is a backup in case fixes cause other issues.
  (if (ScalableFloat|GetValueasBool (Variables|Balance|GetRow_DisableAttachedFX))
    (bind _targetplayer (Variables|FortGameCueNotifyLoop_Cloak|GetTargetPlayer)
      (:then
        (for _array_element (Components|GetChildrenComponents (Variables|/Script/Engine.Character|GetMesh _targetplayer) true)
          (bind _asfxsystem_component (Utilities|Casting|CastToFXSystemComponent _array_element))
          (if (Rendering|IsVisible _asfxsystem_component)
            (Rendering|SetVisibility _asfxsystem_component false false)
            (Utilities|Array|AddUnique (Variables|Default|GetHiddenPlayerFX) _asfxsystem_component))))
      (:else))))

(fn RemoveCloakMaterialOverride
  ;[0.7781 0.3547 0.05928 0.5] Override parent cloak removal to only play FX if fully cloaked and to re-show attach FX
  (Utilities|FlowControl|Sequence
    (:then_0
      (Class|FortPlayerPawn|RemoveMaterialOverrideLocal (Variables|FortGameCueNotifyLoop_Cloak|GetTargetPlayer) (Variables|Default|GetMaterialOverrideId))
      (Guid|InvalidateGuid (Variables|Default|GetMaterialOverrideId))
      (bind _targetplayer (Variables|FortGameCueNotifyLoop_Cloak|GetTargetPlayer)
        (:then
          (if (Variables|Default|GetbCloaked)
            (Niagara|SpawnSystemAttached (Variables|FX|GetDeactivateFXTemplate) (Variables|/Script/Engine.Character|GetMesh _targetplayer) "root" "0, 0, 0" "0, 0, 0" "KeepRelativeOffset" true true "AutoRelease")
            (Variables|Default|SetbCloaked false)))
        (:else)))
    (:then_1
      (Rendering|SetCastShadow (Variables|/Script/Engine.Character|GetMesh (Variables|FortGameCueNotifyLoop_Cloak|GetTargetPlayer)) true)
      (CallFunction|ShowAttachedFX)
      (Variables|FortGameCueNotifyLoop_Cloak|SetTargetPlayer 0))))

(fn ShowAttachedFX
  (bind _hiddenplayerfx (Variables|Default|GetHiddenPlayerFX))
  ;[0.2517 0.06772 0.2906 0.5] Sets FX components that were set to not visible by cloak back to visible
  (bind (_arrayindex _arrayelement) (Utilities|Array|ReverseforEachLoop _hiddenplayerfx)
    (:LoopBody
      (Utilities|FlowControl|Sequence
        (:then_0
          (Utilities|IsValid _arrayelement
            (:Is%20Valid
              (Rendering|SetVisibility _arrayelement true false))))
        (:then_1
          (Utilities|Array|RemoveIndex _hiddenplayerfx _arrayindex))))))

(fn ConstructionScript
  (|CallParentFunction))
